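# Sound cue for the Quad vehicle's boost, apparently the third step of a ring-increase sound, in the vehicles sound class.
# A quality-level root sends the first two levels to a random pick of two Quad fuel-ring waves and the last two to two reused Jetpack fuel-up UI beeps.
# It uses close golf-cart movement attenuation, with slightly raised pitch and slightly lowered volume.
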SoundCue /Game/Athena/Sounds/Vehicles/Quad/Quad_Boost_RingIncrease_3_Cue
root=QualityLevel_0
nodes [
 {id=QualityLevel_0, children [Random_1,Random_1,Random_0,Random_0]}
 {id=Random_0, children [WavePlayer_0,WavePlayer_2], properties {weights [1,1]}}
 {id=Random_1, children [WavePlayer_3,WavePlayer_4], properties {weights [1,1]}}
 {id=WavePlayer_0, properties {soundWaveAssetPtr=/Game/Athena/Sounds/Items/Jetpack/Jetpack_FuelUp_UIBeep_01}}
 {id=WavePlayer_2, properties {soundWaveAssetPtr=/Game/Athena/Sounds/Items/Jetpack/Jetpack_FuelUp_UIBeep_02}}
 {id=WavePlayer_3, properties {soundWaveAssetPtr=&/Quad_FuelRingIncrease_01}}
 {id=WavePlayer_4, properties {soundWaveAssetPtr=&/Quad_FuelRingIncrease_02}}
]
properties {
 attenuationSettings=/Game/Sounds/Attenuation/Fort_GolfCart_Movement_Close_Att
 maxDistance=5300
 pitchMultiplier=1.1
 soundClassObject=&Fort_Vehicles
 volumeMultiplier=0.9
}
