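# Sound cue for the Quad vehicle, apparently the final boost ring increase sound, played in the vehicles sound class.
# A quality level node routes its first two slots to the dedicated Quad fuel ring wave.
# Its last two slots instead play a reused Jetpack fuel-up wave through a modulator that pitches it sharply up and lowers its volume.
# Golf cart close-movement attenuation handles distance falloff.
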SoundCue /Game/Athena/Sounds/Vehicles/Quad/Quad_Boost_RingIncrease_Final_Cue
root=QualityLevel_0
nodes [
 {id=Modulator_0, children [WavePlayer_1], properties {pitchMax=2, pitchMin=1.9, volumeMax=0.7, volumeMin=0.7}}
 {id=QualityLevel_0, children [WavePlayer_0,WavePlayer_0,Modulator_0,Modulator_0]}
 {id=WavePlayer_0, properties {soundWaveAssetPtr=&/Quad_FuelRingIncrease_Final_01}}
 {id=WavePlayer_1, properties {soundWaveAssetPtr=/Game/Athena/Sounds/Items/Jetpack/Jetpack_FuelUp_01}}
]
properties {
 attenuationSettings=/Game/Sounds/Attenuation/Fort_GolfCart_Movement_Close_Att
 maxDistance=5300
 soundClassObject=&Fort_Vehicles
 volumeMultiplier=0.55
}
