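# Sound cue for the driver equip sound of the Quad vehicle in Athena.
# A Mixer layers the golf cart driver equip foley with the Quad engine-on sound, weighting the engine layer louder.
# It borrows the golf cart close-movement attenuation, routes to the vehicles sound class, and overrides concurrency to cap simultaneous instances.
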SoundCue /Game/Athena/Sounds/Vehicles/Quad/Quad_Driver_Equip_Cue
root=Mixer_0
nodes [
 {id=Mixer_0, children [WavePlayer_0,WavePlayer_1], properties {inputVolume [1,2]}}
 {id=WavePlayer_0, properties {soundWaveAssetPtr=/Game/Athena/Sounds/Vehicles/GolfCart/PlayerFoley/GolfCart_Equip_Driver}}
 {id=WavePlayer_1, properties {soundWaveAssetPtr=&/Quad_EngineOn_01}}
]
properties {
 attenuationSettings=/Game/Sounds/Attenuation/Fort_GolfCart_Movement_Close_Att
 bOverrideConcurrency=true
 maxDistance=5300
 soundClassObject=&Fort_Vehicles
 concurrencyOverrides {maxCount=4}
}
