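# Sound cue for the winter variant of the Quad vehicle's horn start, pairing a horn with a Winterfest jingle.
# A Mixer layers two Random nodes at equal volume: one picks from four Winterfest jingles, the other from two horn start waves borrowed from the GolfCart.
# The combined sound is pitched up and lowered in volume, uses the GolfCart close-movement attenuation, and routes to the vehicles sound class.
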SoundCue /Game/Athena/Sounds/Vehicles/Quad/Quad_Horn_Start_Winter_Cue
root=Mixer_0
nodes [
 {id=Mixer_0, children [Random_1,Random_0], properties {inputVolume [1,1]}}
 {id=Random_0, children [WavePlayer_1,WavePlayer_2], properties {weights [1,1]}}
 {id=Random_1, children [WavePlayer_0,WavePlayer_3,WavePlayer_4,WavePlayer_5], properties {weights [1,1,1,1]}}
 {id=WavePlayer_0, properties {soundWaveAssetPtr=&/Winterfest_Jingle_02}}
 {id=WavePlayer_1, properties {soundWaveAssetPtr=/Game/Athena/Sounds/Vehicles/GolfCart/GolfCart_Horn_Start_02}}
 {id=WavePlayer_2, properties {soundWaveAssetPtr=/Game/Athena/Sounds/Vehicles/GolfCart/GolfCart_Horn_Start_01}}
 {id=WavePlayer_3, properties {soundWaveAssetPtr=&/Winterfest_Jingle_03}}
 {id=WavePlayer_4, properties {soundWaveAssetPtr=&/Winterfest_Jingle_01}}
 {id=WavePlayer_5, properties {soundWaveAssetPtr=&/Winterfest_Jingle_04}}
]
properties {
 attenuationSettings=/Game/Sounds/Attenuation/Fort_GolfCart_Movement_Close_Att
 maxDistance=5300
 pitchMultiplier=1.5
 soundClassObject=&Fort_Vehicles
 volumeMultiplier=0.6
}
