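# Looping SoundCue for the Quad vehicle's boost-charge sound, built on a reused frontend UI energy counter loop.
# A branch on bHasDriver plays it only when true, and an enveloper fades the volume in over half a second.
# The Boost parameter raises pitch through a continuous modulator and drives a crossfade that swells volume, then cuts it off near full charge.
# It uses golf cart close-movement attenuation, the vehicles sound class and a concurrency cap.
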SoundCue /Game/Athena/Sounds/Vehicles/Quad/Quad_Movement_BoostCharge_Loop_Cue
root=Branch_0
nodes [
 {id=Branch_0, children [Enveloper_0,null,null], properties {boolParameterName=bHasDriver}}
 {
  id=Enveloper_0
  children [ModulatorContinuous_1]
  properties {pitchCurve {editorCurveData {keys [{time=0}]}}, volumeCurve {editorCurveData {keys [<at 0 0>,<at 0.5 1>]}}}
 }
 {id=ModulatorContinuous_1, children [ParamCrossFade_0], properties {pitchModulationParams {maxOutput=2, minOutput=0.5, parameterName=Boost}}}
 {
  id=ParamCrossFade_0
  children [WavePlayer_1]
  properties {paramName=Boost, crossFadeInput [{fadeInDistanceEnd=0.95, fadeInDistanceStart=0, fadeOutDistanceEnd=1, fadeOutDistanceStart=0.95, volume=1}]}
 }
 {id=WavePlayer_1, properties {bLooping=true, soundWaveAssetPtr=/Game/Sounds/Fort_UI_Sounds/FrontEnd/Counter_Energy_Loop08}}
]
properties {
 attenuationSettings=/Game/Sounds/Attenuation/Fort_GolfCart_Movement_Close_Att
 bOverrideConcurrency=true
 maxDistance=5300
 soundClassObject=&Fort_Vehicles
 volumeMultiplier=0.45
 concurrencyOverrides {maxCount=4}
}
