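# Looping boost movement sound cue for the Quad vehicle, built from pitched-up glider thruster loops in the vehicle sound class.
# A bHasDriver branch plays audio only when true, through an Enveloper that fades volume in.
# A quality-level node mixes both thrust loops for the first two levels and uses only the Shuttle loop for the other two.
# It borrows the golf cart's distant attenuation and overrides concurrency to cap simultaneous instances.
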SoundCue /Game/Athena/Sounds/Vehicles/Quad/Quad_Movement_Boost_Loop_Cue
root=Branch_0
nodes [
 {id=Branch_0, children [Enveloper_0,null,null], properties {boolParameterName=bHasDriver}}
 {
  id=Enveloper_0
  children [QualityLevel_0]
  properties {pitchCurve {editorCurveData {keys [{time=0}]}}, volumeCurve {editorCurveData {keys [<at 0 0>,<at 0.5 1>]}}}
 }
 {id=Mixer_0, children [WavePlayer_0,WavePlayer_2], properties {inputVolume [1,1]}}
 {id=QualityLevel_0, children [Mixer_0,Mixer_0,WavePlayer_2,WavePlayer_2]}
 {id=WavePlayer_0, properties {bLooping=true, soundWaveAssetPtr=/Game/Sounds/Fort_Gadgets/Athena/Glider_CarbideBlack_Thrustloop_01}}
 {id=WavePlayer_2, properties {bLooping=true, soundWaveAssetPtr=/Game/Sounds/Fort_Gadgets/Athena/SFX_Glider_Shuttle_ThrustLoop}}
]
properties {
 attenuationSettings=/Game/Sounds/Attenuation/Fort_GolfCart_Movement_Distant_Att
 bOverrideConcurrency=true
 maxDistance=10100
 pitchMultiplier=1.5
 soundClassObject=&Fort_Vehicles
 volumeMultiplier=0.8
 concurrencyOverrides {maxCount=4}
}
