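# Looping engine idle sound cue for the Quad vehicle, played under the Fort_Vehicles sound class with a raised pitch and reduced volume.
# A quality-level node selects between the standard idle loop and an OPT variant, two levels each.
# The selection feeds an attenuation node that reuses the GolfCart distant movement settings, then a smoothed Doppler node at the root.
# A concurrency override caps simultaneous instances.
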SoundCue /Game/Athena/Sounds/Vehicles/Quad/Quad_Movement_EngineIdle_Cue
root=Doppler_0
nodes [
 {id=Attenuation_0, children [QualityLevel_0], properties {attenuationSettings=/Game/Sounds/Attenuation/Fort_GolfCart_Movement_Distant_Att}}
 {id=Doppler_0, children [Attenuation_0], properties {bUseSmoothing=true}}
 {id=QualityLevel_0, children [WavePlayer_4,WavePlayer_4,WavePlayer_0,WavePlayer_0]}
 {id=WavePlayer_0, properties {bLooping=true, soundWaveAssetPtr=&/Quad_EngineIdle_01_OPT}}
 {id=WavePlayer_4, properties {bLooping=true, soundWaveAssetPtr=&/Quad_EngineIdle_01}}
]
properties {
 bOverrideConcurrency=true
 maxDistance=10100
 pitchMultiplier=1.25
 soundClassObject=&Fort_Vehicles
 volumeMultiplier=0.5
 concurrencyOverrides {maxCount=4}
}
