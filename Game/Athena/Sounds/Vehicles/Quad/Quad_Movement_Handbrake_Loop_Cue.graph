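# Looping sound cue for the Quad vehicle's handbrake skid, built from reused GolfCart tire recordings.
# A Branch on bHasDriver plays sound only on its true path, through an Enveloper that fades volume in over half a second.
# A Mixer layers a pitched-up, boosted tire turn loop with a tire skid loop that a QualityLevel node switches between standard and optimized waves.
# It uses GolfCart close-movement attenuation, the vehicle sound class and a concurrency limit override.
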SoundCue /Game/Athena/Sounds/Vehicles/Quad/Quad_Movement_Handbrake_Loop_Cue
root=Branch_0
nodes [
 {id=Branch_0, children [Enveloper_0,null,null], properties {boolParameterName=bHasDriver}}
 {
  id=Enveloper_0
  children [Mixer_0]
  properties {pitchCurve {editorCurveData {keys [{time=0}]}}, volumeCurve {editorCurveData {keys [<at 0 0>,<at 0.5 1>]}}}
 }
 {id=Mixer_0, children [Modulator_0, QualityLevel_0], properties {inputVolume [1.5,1]}}
 {id=Modulator_0, children [WavePlayer_1], properties {pitchMax=1.5, pitchMin=1.5}}
 {id=QualityLevel_0, children [WavePlayer_3,WavePlayer_3,WavePlayer_0,WavePlayer_0]}
 {id=WavePlayer_0, properties {bLooping=true, soundWaveAssetPtr=/Game/Athena/Sounds/Vehicles/GolfCart/GolfCart_TireSkid_Loop_36_opt}}
 {id=WavePlayer_1, properties {bLooping=true, soundWaveAssetPtr=/Game/Athena/Sounds/Vehicles/GolfCart/GolfCart_Tire_TurnLoop_03}}
 {id=WavePlayer_3, properties {bLooping=true, soundWaveAssetPtr=/Game/Athena/Sounds/Vehicles/GolfCart/GolfCart_TireSkid_Loop_36}}
]
properties {
 attenuationSettings=/Game/Sounds/Attenuation/Fort_GolfCart_Movement_Close_Att
 bOverrideConcurrency=true
 maxDistance=5300
 soundClassObject=&Fort_Vehicles
 volumeMultiplier=0.6
 concurrencyOverrides {maxCount=4}
}
