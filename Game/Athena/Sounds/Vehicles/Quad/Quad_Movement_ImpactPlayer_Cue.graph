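# Sound cue for the Quad vehicle's player-impact sound, routed through the vehicles sound class with close golf cart movement attenuation.
# A quality-level node sends the first two levels to a random pick of three metal barrel impacts, passed through a modulator for slight variation.
# The last two levels use a random pick of six golf cart impacts with no modulator.
# The whole cue plays at a raised pitch.
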SoundCue /Game/Athena/Sounds/Vehicles/Quad/Quad_Movement_ImpactPlayer_Cue
root=QualityLevel_0
nodes [
 {id=Modulator_0, children [Random_0]}
 {id=QualityLevel_0, children [Modulator_0,Modulator_0,Random_1,Random_1]}
 {id=Random_0, children [WavePlayer_4,WavePlayer_2,WavePlayer_3], properties {weights [1,1,1]}}
 {id=Random_1, children [WavePlayer_0,WavePlayer_5,WavePlayer_6,WavePlayer_7,WavePlayer_8,WavePlayer_9], properties {weights [1,1,1,1,1,1]}}
 {id=WavePlayer_0, properties {soundWaveAssetPtr=/Game/Athena/Sounds/Vehicles/GolfCart/GolfCart_Impact_42}}
 {id=WavePlayer_2, properties {soundWaveAssetPtr=/Game/Sounds/Fort_Impact_Sounds/Demolition/Metal/Metal_Barrel_Impact05}}
 {id=WavePlayer_3, properties {soundWaveAssetPtr=/Game/Sounds/Fort_Impact_Sounds/Demolition/Metal/Metal_Barrel_Impact04}}
 {id=WavePlayer_4, properties {soundWaveAssetPtr=/Game/Sounds/Fort_Impact_Sounds/Demolition/Metal/Metal_Barrel_Impact06}}
 {id=WavePlayer_5, properties {soundWaveAssetPtr=/Game/Athena/Sounds/Vehicles/GolfCart/GolfCart_Impact_39}}
 {id=WavePlayer_6, properties {soundWaveAssetPtr=/Game/Athena/Sounds/Vehicles/GolfCart/GolfCart_Impact_40}}
 {id=WavePlayer_7, properties {soundWaveAssetPtr=/Game/Athena/Sounds/Vehicles/GolfCart/GolfCart_Impact_41}}
 {id=WavePlayer_8, properties {soundWaveAssetPtr=/Game/Athena/Sounds/Vehicles/GolfCart/GolfCart_Impact_43}}
 {id=WavePlayer_9, properties {soundWaveAssetPtr=/Game/Athena/Sounds/Vehicles/GolfCart/GolfCart_Impact_44}}
]
properties {
 attenuationSettings=/Game/Sounds/Attenuation/Fort_GolfCart_Movement_Close_Att
 maxDistance=5300
 pitchMultiplier=1.25
 soundClassObject=&Fort_Vehicles
}
