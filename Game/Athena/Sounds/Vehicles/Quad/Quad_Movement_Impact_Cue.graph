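# Sound cue for Quad vehicle movement impacts, built from reused GolfCart impact samples.
# A Random node picks one of six equally weighted GolfCart impact waves, played with raised pitch and slightly boosted volume.
# It uses the close GolfCart movement attenuation and the vehicles sound class, with a concurrency override capping simultaneous instances at four.
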SoundCue /Game/Athena/Sounds/Vehicles/Quad/Quad_Movement_Impact_Cue
root=Random_1
nodes [
 {id=Random_1, children [WavePlayer_0,WavePlayer_5,WavePlayer_6,WavePlayer_7,WavePlayer_8,WavePlayer_9], properties {weights [1,1,1,1,1,1]}}
 {id=WavePlayer_0, properties {soundWaveAssetPtr=/Game/Athena/Sounds/Vehicles/GolfCart/GolfCart_Impact_42}}
 {id=WavePlayer_5, properties {soundWaveAssetPtr=/Game/Athena/Sounds/Vehicles/GolfCart/GolfCart_Impact_39}}
 {id=WavePlayer_6, properties {soundWaveAssetPtr=/Game/Athena/Sounds/Vehicles/GolfCart/GolfCart_Impact_40}}
 {id=WavePlayer_7, properties {soundWaveAssetPtr=/Game/Athena/Sounds/Vehicles/GolfCart/GolfCart_Impact_41}}
 {id=WavePlayer_8, properties {soundWaveAssetPtr=/Game/Athena/Sounds/Vehicles/GolfCart/GolfCart_Impact_43}}
 {id=WavePlayer_9, properties {soundWaveAssetPtr=/Game/Athena/Sounds/Vehicles/GolfCart/GolfCart_Impact_44}}
]
properties {
 attenuationSettings=/Game/Sounds/Attenuation/Fort_GolfCart_Movement_Close_Att
 bOverrideConcurrency=true
 maxDistance=5300
 pitchMultiplier=1.5
 soundClassObject=&Fort_Vehicles
 volumeMultiplier=1.2
 concurrencyOverrides {maxCount=4}
}
