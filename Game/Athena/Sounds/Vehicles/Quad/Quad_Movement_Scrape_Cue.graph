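# Sound cue for the looping scrape sound of the Quad vehicle's movement, routed to the vehicles sound class.
# It reuses the GolfCart scrape loop and close movement attenuation, with raised pitch and reduced volume.
# A quality level node at the root feeds that one looping wave into its first two slots and leaves the other two empty.
# A concurrency override limits simultaneous instances.
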SoundCue /Game/Athena/Sounds/Vehicles/Quad/Quad_Movement_Scrape_Cue
root=QualityLevel_0
nodes [
 {id=QualityLevel_0, children [WavePlayer_2,WavePlayer_2,null,null]}
 {id=WavePlayer_2, properties {bLooping=true, soundWaveAssetPtr=/Game/Athena/Sounds/Vehicles/GolfCart/GolfCart_ScrapeLoop}}
]
properties {
 attenuationSettings=/Game/Sounds/Attenuation/Fort_GolfCart_Movement_Close_Att
 bOverrideConcurrency=true
 maxDistance=5300
 pitchMultiplier=1.2
 soundClassObject=&Fort_Vehicles
 volumeMultiplier=0.6
 concurrencyOverrides {maxCount=3}
}
