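# Looping movement wind sound cue for the Quad vehicle, audible only when the bHasDriver parameter is true.
# An Enveloper fades the wind in over half a second.
# A quality switch feeds it either a mix of sliding-wind and umbrella-closed falling loops for the first two levels, or a single optimized umbrella-closed loop for the other two.
# It borrows golf cart distant attenuation, uses the vehicle sound class and caps concurrent instances.
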SoundCue /Game/Athena/Sounds/Vehicles/Quad/Quad_Movement_Wind_Loop_Cue
root=Branch_0
nodes [
 {id=Branch_0, children [Enveloper_0,null,null], properties {boolParameterName=bHasDriver}}
 {
  id=Enveloper_0
  children [QualityLevel_0]
  properties {pitchCurve {editorCurveData {keys [{time=0}]}}, volumeCurve {editorCurveData {keys [<at 0 0>,<at 0.5 1>]}}}
 }
 {id=Mixer_0, children [WavePlayer_1,WavePlayer_4], properties {inputVolume [1,1]}}
 {id=QualityLevel_0, children [Mixer_0,Mixer_0,WavePlayer_0,WavePlayer_0]}
 {id=WavePlayer_0, properties {bLooping=true, soundWaveAssetPtr=/Game/Athena/Sounds/Wind_Falling/Wind_UmbrellaClosed_Loop_1P_opt_02}}
 {id=WavePlayer_1, properties {bLooping=true, soundWaveAssetPtr=/Game/Athena/Sounds/Character/Sliding/Waves/sfx_character_slidingloop_wind}}
 {id=WavePlayer_4, properties {bLooping=true, soundWaveAssetPtr=/Game/Athena/Sounds/Wind_Falling/Wind_UmbrellaClosed_Loop_01}}
]
properties {
 attenuationSettings=/Game/Sounds/Attenuation/Fort_GolfCart_Movement_Distant_Att
 bOverrideConcurrency=true
 maxDistance=10100
 soundClassObject=&Fort_Vehicles
 volumeMultiplier=0.65
 concurrencyOverrides {maxCount=4}
}
