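# Sound cue for the equip foley heard when a rear passenger on the Quad vehicle equips an item.
# A Random node picks one of three equally weighted wave players, reusing the GolfCart rear equip recordings rather than Quad-specific audio.
# It plays at reduced volume in the vehicles sound class, spatialized through a close-range GolfCart movement attenuation asset.
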SoundCue /Game/Athena/Sounds/Vehicles/Quad/Quad_RearPassenger_Equip_Cue
root=Random_0
nodes [
 {id=Random_0, children [WavePlayer_1,WavePlayer_2,WavePlayer_3], properties {weights [1,1,1]}}
 {id=WavePlayer_1, properties {soundWaveAssetPtr=/Game/Athena/Sounds/Vehicles/GolfCart/PlayerFoley/GolfCart_Equip_Rear_01}}
 {id=WavePlayer_2, properties {soundWaveAssetPtr=/Game/Athena/Sounds/Vehicles/GolfCart/PlayerFoley/GolfCart_Equip_Rear_02}}
 {id=WavePlayer_3, properties {soundWaveAssetPtr=/Game/Athena/Sounds/Vehicles/GolfCart/PlayerFoley/GolfCart_Equip_Rear_03}}
]
properties {
 attenuationSettings=/Game/Sounds/Attenuation/Fort_GolfCart_Movement_Close_Att
 maxDistance=5300
 soundClassObject=&Fort_Vehicles
 volumeMultiplier=0.3
}
