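# Sound cue for quad vehicle suspension noise, built from six reused GolfCart suspension recordings with raised overall pitch and volume.
# An equally weighted random node picks one recording and feeds a modulator that varies pitch slightly while holding volume constant.
# It plays as player foley with GolfCart close-movement attenuation, and a concurrency override blocks new instances once its limit is reached.
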SoundCue /Game/Athena/Sounds/Vehicles/Quad/Quad_Suspension_Cue
root=Modulator_1
nodes [
 {id=Modulator_1, children [Random_0], properties {pitchMax=1, volumeMax=1, volumeMin=1}}
 {
  id=Random_0
  children [WavePlayer_9, WavePlayer_10, WavePlayer_11, WavePlayer_12, WavePlayer_13, WavePlayer_14]
  properties {weights [1,1,1,1,1,1]}
 }
 {id=WavePlayer_10, properties {soundWaveAssetPtr=/Game/Athena/Sounds/Vehicles/GolfCart/GolfCart_Suspension_51}}
 {id=WavePlayer_11, properties {soundWaveAssetPtr=/Game/Athena/Sounds/Vehicles/GolfCart/GolfCart_Suspension_48}}
 {id=WavePlayer_12, properties {soundWaveAssetPtr=/Game/Athena/Sounds/Vehicles/GolfCart/GolfCart_Suspension_49}}
 {id=WavePlayer_13, properties {soundWaveAssetPtr=/Game/Athena/Sounds/Vehicles/GolfCart/GolfCart_Suspension_50}}
 {id=WavePlayer_14, properties {soundWaveAssetPtr=/Game/Athena/Sounds/Vehicles/GolfCart/GolfCart_Suspension_47}}
 {id=WavePlayer_9, properties {soundWaveAssetPtr=/Game/Athena/Sounds/Vehicles/GolfCart/GolfCart_Suspension_52}}
]
properties {
 attenuationSettings=/Game/Sounds/Attenuation/Fort_GolfCart_Movement_Close_Att
 bOverrideConcurrency=true
 maxDistance=5300
 pitchMultiplier=1.5
 soundClassObject=&Fort_Player_Foley
 volumeMultiplier=1.5
 attenuationOverrides {falloffDistance=1000, attenuationShapeExtents {x=500}}
 concurrencyOverrides {maxCount=8, resolutionRule=PreventNew}
}
