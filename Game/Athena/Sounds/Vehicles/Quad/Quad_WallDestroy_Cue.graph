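# Sound cue for the Quad vehicle destroying a wall, reusing a Carmine punch blast sound in the vehicle sound class.
# A quality-level root sends its first two slots to the standard wave with randomized pitch and volume, and its last two to an "_opt" wave with fixed pitch.
# It borrows close golf-cart movement attenuation and overrides concurrency to two voices, stopping the oldest.
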SoundCue /Game/Athena/Sounds/Vehicles/Quad/Quad_WallDestroy_Cue
root=QualityLevel_0
nodes [
 {id=Modulator_0, children [WavePlayer_2], properties {pitchMax=1.1, pitchMin=0.85, volumeMin=0.7}}
 {id=Modulator_2, children [WavePlayer_0], properties {pitchMax=1, pitchMin=1}}
 {id=QualityLevel_0, children [Modulator_0,Modulator_0,Modulator_2,Modulator_2]}
 {id=WavePlayer_0, properties {soundWaveAssetPtr=/Game/Sounds/Fort_GamePlay_Sounds/Carmine/Carmine_Punch_Blast_02_opt}}
 {id=WavePlayer_2, properties {soundWaveAssetPtr=/Game/Sounds/Fort_GamePlay_Sounds/Carmine/Carmine_Punch_Blast_02}}
]
properties {
 attenuationSettings=/Game/Sounds/Attenuation/Fort_GolfCart_Movement_Close_Att
 bOverrideConcurrency=true
 maxDistance=5300
 soundClassObject=&Fort_Vehicles
 volumeMultiplier=0.6
 concurrencyOverrides {maxCount=2, resolutionRule=StopOldest, volumeScale=0.5}
}
